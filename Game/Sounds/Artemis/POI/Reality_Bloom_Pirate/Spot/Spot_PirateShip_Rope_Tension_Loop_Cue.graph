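# Looping spot ambience cue for creaking rope tension on the Reality Bloom pirate ship POI.
# A random choice of two looping rope waves feeds three nested looping envelopers that slowly drift pitch, swell and fade volume in irregular bursts, and add a gentler periodic dip.
# A short random start delay and a quality switch, silent on two of four levels, sit above that chain.
# Capsule attenuation, single-voice concurrency and an exterior ambient bus send shape playback.
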SoundCue /Game/Sounds/Artemis/POI/Reality_Bloom_Pirate/Spot/Spot_PirateShip_Rope_Tension_Loop_Cue
root=Delay_0
nodes [
 {id=Delay_0, children [QualityLevel_0], properties {delayMax=0.4}}
 {
  id=Enveloper_0
  children [Random_0]
  properties {
   bLoop=true
   bLoopIndefinitely=true
   loopEnd=60
   pitchCurve {editorCurveData {keys [
      {interpMode=RCIM_Cubic, time=0}
      <at 15 0.85 RCIM_Cubic>
      {arriveTangent=0.0091, interpMode=RCIM_Cubic, leaveTangent=0.0091, tangentMode=RCTM_User, time=30}
      <at 45 1.2 RCIM_Cubic>
      {time=60}
     ]}}
   volumeCurve {editorCurveData {keys [<at 0 1>]}}
  }
 }
 {
  id=Enveloper_2
  children [Enveloper_3]
  properties {
   bLoop=true
   bLoopIndefinitely=true
   loopEnd=20
   pitchCurve {editorCurveData {keys [{time=0}]}}
   volumeCurve {editorCurveData {keys [
      {arriveTangent=-0.01, interpMode=RCIM_Cubic, time=0, value=1}
      <at 10 0.6 RCIM_Cubic>
      {interpMode=RCIM_Cubic, leaveTangent=0.01, time=20, value=1}
     ]}}
  }
 }
 {
  id=Enveloper_3
  children [Enveloper_0]
  properties {
   bLoop=true
   bLoopIndefinitely=true
   loopEnd=60
   pitchCurve {editorCurveData {keys [{time=0}]}}
   volumeCurve {editorCurveData {keys [
      {arriveTangent=-0.01, interpMode=RCIM_Cubic, time=0, value=1}
      {arriveTangent=-0.0035, interpMode=RCIM_Cubic, leaveTangent=-0.0035, tangentMode=RCTM_User, time=4.3108, value=0.001}
      {arriveTangent=0.0016, interpMode=RCIM_Cubic, leaveTangent=0.0016, tangentMode=RCTM_User, time=7.2345, value=0.7106}
      {arriveTangent=0.0008, interpMode=RCIM_Cubic, leaveTangent=0.0008, tangentMode=RCTM_User, time=10.8757, value=0.0144}
      {arriveTangent=0.0016, interpMode=RCIM_Cubic, leaveTangent=0.0016, tangentMode=RCTM_User, time=12.7454, value=0}
      {arriveTangent=0.0009, interpMode=RCIM_Cubic, leaveTangent=0.0009, time=14.812, value=0.593}
      {arriveTangent=-0.0011, interpMode=RCIM_Cubic, leaveTangent=-0.0011, tangentMode=RCTM_User, time=18.3548, value=0.005}
      {arriveTangent=0.0002, interpMode=RCIM_Cubic, leaveTangent=0.0002, time=23.8657, value=0.8146}
      {arriveTangent=0.0016, interpMode=RCIM_Cubic, leaveTangent=0.0016, tangentMode=RCTM_User, time=27.802, value=0.0072}
      {arriveTangent=-0.0003, interpMode=RCIM_Cubic, leaveTangent=-0.0003, time=31.0495, value=0.6077}
      {arriveTangent=0.0009, interpMode=RCIM_Cubic, leaveTangent=0.0009, tangentMode=RCTM_User, time=34.6907, value=0.0052}
      {arriveTangent=-0.0008, interpMode=RCIM_Cubic, leaveTangent=-0.0008, time=37.8398, value=0.8146}
      <at 41.7761 -0.0001>
      {arriveTangent=0.0024, interpMode=RCIM_Cubic, leaveTangent=0.0024, tangentMode=RCTM_User, time=44.3348, value=0.005}
      {arriveTangent=-0.0006, interpMode=RCIM_Cubic, leaveTangent=-0.0006, time=47.7791, value=0.7106}
      {arriveTangent=-0.0056, interpMode=RCIM_Cubic, leaveTangent=-0.0056, tangentMode=RCTM_User, time=52.9022, value=0}
      <at 55.1597 0.3798>
      {arriveTangent=-0.0003, interpMode=RCIM_Cubic, leaveTangent=-0.0003, tangentMode=RCTM_User, time=57.5216, value=-0.0003}
      <at 60 1 RCIM_Cubic>
     ]}}
  }
 }
 {id=QualityLevel_0, children [Enveloper_2,Enveloper_2,null,null]}
 {id=Random_0, children [WavePlayer_2,WavePlayer_0], properties {weights [1,1]}}
 {
  id=WavePlayer_0
  properties {bLooping=true, soundWaveAssetPtr=&/Spot_PirateShip_Rope_Tension_Loop_02}
 }
 {
  id=WavePlayer_2
  properties {bLooping=true, soundWaveAssetPtr=&/Spot_PirateShip_Rope_Tension_Loop_01}
 }
]
properties {
 bOverrideAttenuation=true
 bOverrideConcurrency=true
 maxDistance=1400
 soundClassObject=&Fort_Gadgets
 volumeMultiplier=0.3
 attenuationOverrides {
  attenuationShape=Capsule
  bEnableListenerFocus=true
  distanceAlgorithm=Custom
  falloffDistance=750
  focusVolumeAttenuation=1.1
  lPFFrequencyAtMax=1000
  lPFRadiusMax=2000
  lPFRadiusMin=512
  nonFocusVolumeAttenuation=0.75
  attenuationShapeExtents {x=650}
  customAttenuationCurve {editorCurveData {keys [{arriveTangent=-0.3131, interpMode=RCIM_Cubic, leaveTangent=-0.3131, tangentMode=RCTM_User, time=0, value=1}, <at 1 0>]}}
 }
 busSends [{audioBus=null, minSendLevel=0, sendLevel=1, soundSourceBus=&AmbientAudio_Exterior_SourceBus}]
 concurrencyOverrides {maxCount=1, voiceStealReleaseTime=3}
}
