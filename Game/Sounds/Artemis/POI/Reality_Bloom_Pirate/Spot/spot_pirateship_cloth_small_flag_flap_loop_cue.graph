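# Looping ambient sound cue for a small cloth flag flapping on the pirate ship spot in the Reality Bloom Pirate POI.
# It reuses the Creative CTF flag loop wave, boosted through a mixer into two nested, endlessly looping envelopers.
# The inner enveloper slowly varies pitch and the outer dips volume on a shorter cycle, disguising repetition.
# It plays quietly as outdoor ambience with overridden logarithmic attenuation and listener focus.
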SoundCue /Game/Sounds/Artemis/POI/Reality_Bloom_Pirate/Spot/spot_pirateship_cloth_small_flag_flap_loop_cue.Spot_PirateShip_Cloth_Small_Flag_Flap_Loop_Cue
root=Enveloper_2
nodes [
 {
  id=Enveloper_1
  children [Mixer_0]
  properties {
   bLoop=true
   bLoopIndefinitely=true
   loopEnd=60
   pitchCurve {editorCurveData {keys [
      {interpMode=RCIM_Cubic, time=0}
      <at 15 0.95 RCIM_Cubic>
      {arriveTangent=0.0072, interpMode=RCIM_Cubic, leaveTangent=0.0072, tangentMode=RCTM_User, time=30}
      <at 45 1.1 RCIM_Cubic>
      {time=60}
     ]}}
   volumeCurve {editorCurveData {keys [<at 0 1>]}}
  }
 }
 {
  id=Enveloper_2
  children [Enveloper_1]
  properties {
   bLoop=true
   bLoopIndefinitely=true
   loopEnd=20
   pitchCurve {editorCurveData {keys [{time=0}]}}
   volumeCurve {editorCurveData {keys [
      {arriveTangent=-0.01, interpMode=RCIM_Cubic, time=0, value=1}
      <at 10 0.8 RCIM_Cubic>
      {interpMode=RCIM_Cubic, leaveTangent=0.01, time=20, value=1}
     ]}}
  }
 }
 {id=Mixer_0, children [WavePlayer_0], properties {inputVolume [3]}}
 {id=WavePlayer_0, properties {bLooping=true, soundWaveAssetPtr=/Game/Sounds/Creative/Modes/CTF/CTF_FlagLoop_01}}
]
properties {
 bOverrideAttenuation=true
 maxDistance=1600
 soundClassObject=&Fort_Ambient_Outdoors_Athena
 volumeMultiplier=0.15
 attenuationOverrides {
  bEnableFocusInterpolation=true
  bEnableListenerFocus=true
  dBAttenuationAtMax=-40
  distanceAlgorithm=Logarithmic
  falloffDistance=1500
  focusAzimuth=15
  focusVolumeAttenuation=1.2
  nonFocusAzimuth=45
  nonFocusVolumeAttenuation=0.4
  nonSpatializedRadiusStart=100
  occlusionTraceChannel=ECC_GameTraceChannel11
  occlusionVolumeAttenuation=0.7
  stereoSpread=2000
  attenuationShapeExtents {x=100}
 }
}
